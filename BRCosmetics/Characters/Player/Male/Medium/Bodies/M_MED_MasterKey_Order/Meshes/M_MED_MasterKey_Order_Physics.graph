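# Physics asset for the MasterKey Order medium male player body, covering only the pelvis, thighs, back coat tail and two side pouches.
# A box shapes the coat tail, two capsules shape each pouch, and single capsules cover the pelvis and each thigh.
# All constraints attach to the pelvis: coat tail and pouches have limited swing and tight twist, while the thighs rotate freely.
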PhysicsAsset /BRCosmetics/Characters/Player/Male/Medium/Bodies/M_MED_MasterKey_Order/Meshes/M_MED_MasterKey_Order_Physics
bodies {
 dyn_coat_tail_back [{shapeType=Box, extentX=17.6337, extentY=21.1977, extentZ=4.4788, center=<7.0809 0 1.5344>, rotation=<rot -2.8161 0 0>}]
 pelvis [{shapeType=Capsule, length=0, radius=12}]
 pouch_l [
  {shapeType=Capsule, length=7.895, radius=2.1122, shapeName=pouch_l_capsule, center=<8.6494 -0.1451 4>, rotation=<rot 0 0 90>}
  {shapeType=Capsule, length=8.2779, radius=1.7293, shapeName=pouch_l_capsule, center=<-1.5028 -0.1451 1.3717>, rotation=<rot 0 0 90>}
 ]
 pouch_r [
  {shapeType=Capsule, length=6.4414, radius=2.7602, shapeName=pouch_r_capsule, center=<8.6038 -0.7688 -4.5974>, rotation=<rot 0 0 90>}
  {
   shapeType=Capsule
   length=6.4414
   radius=2.7602
   shapeName=pouch_r_capsule
   center=<-2.4721 -0.6575 -0.7538>
   rotation=<rot 0 0.0001 96.0113>
  }
 ]
 thigh_l [{shapeType=Capsule, length=62.2823, radius=9.5348, shapeName=thigh_l_capsule, center=<-23.0979 0 0>, rotation=<rot 86.506 179.9998 179.9999>}]
 thigh_r [{shapeType=Capsule, length=30, radius=9.9299, shapeName=thigh_r_capsule, center=<22.0374 0 0>, rotation=<rot 90 0 0>}]
}
constraints [
 {bone1Name=dyn_coat_tail_back, bone2Name=pelvis, swing1LimitDegrees=5, swing2LimitDegrees=35, twistLimitDegrees=3}
 {bone1Name=pouch_l, bone2Name=pelvis, swing1LimitDegrees=45, swing2LimitDegrees=45, twistLimitDegrees=5}
 {bone1Name=pouch_r, bone2Name=pelvis, swing1LimitDegrees=45, swing2LimitDegrees=45, twistLimitDegrees=5}
 {
  bone1Name=thigh_l
  bone2Name=pelvis
  swing1LimitDegrees=45
  swing1Motion=Free
  swing2LimitDegrees=45
  swing2Motion=Free
  twistLimitDegrees=45
  twistMotion=Free
 }
 {
  bone1Name=thigh_r
  bone2Name=pelvis
  swing1LimitDegrees=45
  swing1Motion=Free
  swing2LimitDegrees=45
  swing2Motion=Free
  twistLimitDegrees=45
  twistMotion=Free
 }
]
