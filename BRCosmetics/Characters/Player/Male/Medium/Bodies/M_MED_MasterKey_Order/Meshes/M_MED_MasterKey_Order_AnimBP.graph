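# Post-process style AnimBP for the MasterKey Order male body cosmetic, adding secondary motion to pouches, collar and coat tails.
# It copies the parent mesh pose, runs pelvis-space RigidBody physics, then layers additive bone offsets.
# Event Graph derives collar and front coat-tail alphas from head and thigh rotations, plus backward movement while crouched.
# Pouch constraint, swing and jump nodes are disconnected from the output, leaving only fixed pouch offsets active.
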AnimBlueprint /BRCosmetics/Characters/Player/Male/Medium/Bodies/M_MED_MasterKey_Order/Meshes/M_MED_MasterKey_Order_AnimBP
target_skeleton=&/M_MED_MasterKey_Order_Skeleton
graphs [
 AnimGraph {nodes [
   n1 = "Constraint - pouch_l (R)"!(ComponentPose, bAlphaBoolEnabled:bool=True, ConstraintWeights_0=1) -> (Pose) {
    alphaInputType=Bool
    alphaBoolBlend {blendOption=Linear}
    boneToModify {boneName=pouch_l}
    constraintSetup [{transformType=Rotation, perAxis {bX=false}, targetBone {boneName=thigh_l}}]
    constraintWeights [1]
   }
   "Constraint - pouch_r (R)"!(ComponentPose=@n1, bAlphaBoolEnabled:bool=True, ConstraintWeights_0=1) -> (Pose) {
    alphaInputType=Bool
    alphaBoolBlend {blendOption=Linear}
    boneToModify {boneName=pouch_r}
    constraintSetup [{transformType=Rotation, targetBone {boneName=thigh_r}}]
    constraintWeights [1]
   }
   "Constraint - pouch_l (R)"(bAlphaBoolEnabled=True, ConstraintWeights_0=1) {
    alphaInputType=Bool
    alphaBoolBlend {blendOption=Linear}
    boneToModify {boneName=pouch_l}
    constraintSetup [{transformType=Rotation, targetBone {boneName=thigh_l}}]
    constraintWeights [1]
   }
   "Constraint - pouch_r (R)"(bAlphaBoolEnabled=True, ConstraintWeights_0=1) {
    alphaInputType=Bool
    alphaBoolBlend {blendOption=Linear}
    boneToModify {boneName=pouch_r}
    constraintSetup [{transformType=Rotation, targetBone {boneName=thigh_r}}]
    constraintWeights [1]
   }
   n2 = Component To Local()
   n3 = Copy Pose From Mesh() {bUseAttachedParent=true}
   Local To Component(LocalPose=Blend Poses by bool(
     BlendPose_0=Copy Pose From Mesh() {bUseAttachedParent=true}
     BlendPose_1=Blend Poses by bool(BlendPose_0=@n3, BlendPose_1=@n2) {blendType=Linear}
    ) {blendType=Linear})
   n4 = "Transform (Modify) Bone - Bone: pouch_r"(Alpha=Get Pouch_swing_R() -> (Pouch_swing_R:double), Translation="0,0,-1.5", Rotation="-14,5,15", Scale="1,1,1") {
    rotationMode=BMM_Additive
    rotationSpace=BCS_BoneSpace
    translationMode=BMM_Additive
    translationSpace=BCS_BoneSpace
    alphaBoolBlend {blendOption=Linear}
    alphaScaleBiasClamp {interpSpeedDecreasing=8, interpSpeedIncreasing=9}
    boneToModify {boneName=pouch_r}
    rotation=<rot -14 5 15>
    translation {z=-1.5}
   }
   "Transform (Modify) Bone - Bone: pouch_l"(ComponentPose=@n4, Alpha=Get Pouch_swing_L() -> (Pouch_swing_L:double), Translation="0,0,1.5", Rotation="10,-10,-30", Scale="1,1,1") {
    rotationMode=BMM_Additive
    rotationSpace=BCS_BoneSpace
    translationMode=BMM_Additive
    translationSpace=BCS_BoneSpace
    alphaBoolBlend {blendOption=Linear}
    alphaScaleBiasClamp {interpSpeedDecreasing=8, interpSpeedIncreasing=9}
    boneToModify {boneName=pouch_l}
    rotation=<rot 10 -10 -30>
    translation {z=1.5}
   }
   n5 = Get was_jumping() -> (was_jumping:bool)
   n6 = "Transform (Modify) Bone - Bone: pouch_l"!(ComponentPose, bAlphaBoolEnabled:bool=@n5, Translation="-1,0,0", Rotation="10,0,0", Scale="1,1,1") -> (Pose) {
    alphaInputType=Bool
    rotationMode=BMM_Additive
    rotationSpace=BCS_BoneSpace
    scaleMode=BMM_Additive
    scaleSpace=BCS_BoneSpace
    translationMode=BMM_Additive
    translationSpace=BCS_BoneSpace
    alphaBoolBlend {blendOption=ExpInOut}
    alphaScaleBiasClamp {bInterpResult=true}
    boneToModify {boneName=pouch_l}
    rotation {pitch=10}
    translation {x=-1}
   }
   "Transform (Modify) Bone - Bone: pouch_r"!(ComponentPose=@n6, bAlphaBoolEnabled:bool=@n5, Translation="-1,0,0", Rotation="10,0,0", Scale="1,1,1") -> (Pose) {
    alphaInputType=Bool
    rotationMode=BMM_Additive
    rotationSpace=BCS_BoneSpace
    translationMode=BMM_Additive
    translationSpace=BCS_BoneSpace
    alphaBoolBlend {blendOption=ExpInOut}
    alphaScaleBiasClamp {interpSpeedDecreasing=8, interpSpeedIncreasing=9}
    boneToModify {boneName=pouch_r}
    rotation {pitch=10}
    translation {x=-1}
   }
   n7 = RigidBody(ComponentPose=Local To Component(LocalPose=Copy Pose From Mesh() {bUseAttachedParent=true})) {
    bClampLinearTranslationLimitToRefPose=true
    bOverrideWorldGravity=true
    overridePhysicsAsset=&/M_MED_MasterKey_Order_Physics
    simulationSpace=BaseBoneSpace
    alphaBoolBlend {blendOption=Linear}
    baseBoneRef {boneName=pelvis}
    componentAppliedLinearAccClamp=<1500 1500 1500>
    componentLinearAccScale=<0.2 0.2 0.2>
    componentLinearVelScale=<0.2 0.2 0.2>
    overrideWorldGravity {z=-980}
    simSpaceSettings {
     maxAngularAcceleration=1000
     maxAngularVelocity=3
     maxLinearAcceleration=700
     maxLinearVelocity=500
     velocityScaleZ=0
     worldAlpha=0.5
     externalLinearDragV=<0.2 0.2>
    }
   }
   n8 = "Transform (Modify) Bone - Bone: pouch_l"(ComponentPose=@n7, bAlphaBoolEnabled=True, Translation="0.034,-1.585,0.819", Rotation="-5.315,-6.123,-14.373", Scale="1,1,1") {
    alphaInputType=Bool
    rotationMode=BMM_Additive
    rotationSpace=BCS_BoneSpace
    scaleMode=BMM_Additive
    scaleSpace=BCS_BoneSpace
    translationMode=BMM_Additive
    translationSpace=BCS_BoneSpace
    alphaBoolBlend {blendInTime=0.1, blendOption=Linear, blendOutTime=0.1}
    boneToModify {boneName=pouch_l}
    rotation=<rot -5.315 -6.123 -14.373>
    translation=<0.034 -1.585 0.819>
   }
   n9 = "Transform (Modify) Bone - Bone: pouch_r"(ComponentPose=@n8, bAlphaBoolEnabled=True, Translation="0,0,-0.5", Rotation="0.011,0.014,0", Scale="1,1,1") {
    alphaInputType=Bool
    rotationMode=BMM_Additive
    rotationSpace=BCS_BoneSpace
    scaleMode=BMM_Additive
    scaleSpace=BCS_BoneSpace
    translationMode=BMM_Additive
    translationSpace=BCS_BoneSpace
    alphaBoolBlend {blendInTime=0.1, blendOption=Linear, blendOutTime=0.1}
    boneToModify {boneName=pouch_r}
    rotation {pitch=0.011, yaw=0.014}
    translation {z=-0.5}
   }
   n10 = "Transform (Modify) Bone - Bone: pouch_l"(ComponentPose=@n9, bAlphaBoolEnabled=True, Translation="-0.034,1.585,-0.819", Rotation="7.076,3.953,-8", Scale="1,1,1") {
    alphaInputType=Bool
    rotationMode=BMM_Additive
    rotationSpace=BCS_BoneSpace
    scaleMode=BMM_Additive
    scaleSpace=BCS_BoneSpace
    translationMode=BMM_Additive
    translationSpace=BCS_BoneSpace
    alphaBoolBlend {blendInTime=0.1, blendOption=Linear, blendOutTime=0.1}
    boneToModify {boneName=pouch_l}
    rotation=<rot 7.076 3.953 -8>
    translation=<-0.034 1.585 -0.819>
   }
   n11 = "Transform (Modify) Bone - Bone: pouch_r"(ComponentPose=@n10, bAlphaBoolEnabled=True, Translation="0,0,0", Rotation="0,0,-10", Scale="1,1,1") {
    alphaInputType=Bool
    rotationMode=BMM_Additive
    rotationSpace=BCS_BoneSpace
    scaleMode=BMM_Additive
    scaleSpace=BCS_BoneSpace
    translationMode=BMM_Additive
    translationSpace=BCS_BoneSpace
    alphaBoolBlend {blendInTime=0.1, blendOption=Linear, blendOutTime=0.1}
    boneToModify {boneName=pouch_r}
    rotation {roll=-10}
   }
   n12 = "Transform (Modify) Bone - Bone: pouch_r"(ComponentPose=@n11, bAlphaBoolEnabled=True, Translation="0,0,0", Rotation="0,0,-10", Scale="1,1,1") {
    alphaInputType=Bool
    rotationMode=BMM_Additive
    rotationSpace=BCS_BoneSpace
    scaleMode=BMM_Additive
    scaleSpace=BCS_BoneSpace
    translationMode=BMM_Additive
    translationSpace=BCS_BoneSpace
    alphaBoolBlend {blendInTime=0.1, blendOption=Linear, blendOutTime=0.1}
    boneToModify {boneName=pouch_r}
    rotation {roll=-10}
   }
   n13 = "Transform (Modify) Bone - Bone: dyn_neck_03"(ComponentPose=@n12, Alpha=Get Collar_fix() -> (Collar_fix:double), Translation="0,4,0", Rotation="0,0,0", Scale="1,1,1") {
    rotationMode=BMM_Additive
    rotationSpace=BCS_BoneSpace
    scaleMode=BMM_Additive
    scaleSpace=BCS_BoneSpace
    translationMode=BMM_Additive
    translationSpace=BCS_ParentBoneSpace
    alphaBoolBlend {blendOption=Linear}
    alphaScaleBiasClamp {bInterpResult=true}
    boneToModify {boneName=dyn_neck_03}
    translation {y=4}
   }
   n14 = Get LegSwingDeltaL() -> (LegSwingDeltaL:double)
   n15 = "Transform (Modify) Bone - Bone: dyn_coat_tail_front_l_01"(ComponentPose=@n13, Alpha=@n14, Translation="-6,0,0", Rotation="75,-10,-15", Scale="1,1,1") {
    rotationMode=BMM_Additive
    rotationSpace=BCS_BoneSpace
    scaleMode=BMM_Additive
    scaleSpace=BCS_BoneSpace
    translationMode=BMM_Additive
    translationSpace=BCS_BoneSpace
    alphaBoolBlend {blendOption=Linear}
    alphaScaleBiasClamp {bInterpResult=true, interpSpeedDecreasing=100, interpSpeedIncreasing=100}
    boneToModify {boneName=dyn_coat_tail_front_l_01}
    rotation=<rot 75 -10 -15>
    translation {x=-6}
   }
   n16 = Get Bwd Movement() -> (Bwd Movement:double)
   n17 = "Transform (Modify) Bone - Bone: dyn_coat_tail_front_l_01"(ComponentPose=@n15, Alpha=@n16, Translation="-0.5,0,0", Rotation="10,0,0", Scale="1,1,1") {
    rotationMode=BMM_Additive
    rotationSpace=BCS_BoneSpace
    scaleMode=BMM_Additive
    scaleSpace=BCS_BoneSpace
    translationMode=BMM_Additive
    translationSpace=BCS_BoneSpace
    alphaBoolBlend {blendOption=Linear}
    alphaScaleBiasClamp {bInterpResult=true}
    boneToModify {boneName=dyn_coat_tail_front_l_01}
    rotation {pitch=10}
    translation {x=-0.5}
   }
   n18 = "Transform (Modify) Bone - Bone: dyn_coat_tail_front_l_02"(ComponentPose=@n17, Alpha=@n14, Translation="0,0,0", Rotation="20,0,0", Scale="1,1,1") {
    rotationMode=BMM_Additive
    rotationSpace=BCS_BoneSpace
    translationMode=BMM_Additive
    translationSpace=BCS_BoneSpace
    alphaBoolBlend {blendOption=Linear}
    boneToModify {boneName=dyn_coat_tail_front_l_02}
    rotation {pitch=20}
   }
   n19 = "Transform (Modify) Bone - Bone: dyn_coat_tair_front_r_01"(ComponentPose=@n18, Alpha=@n16, Translation="-0.5,0,0", Rotation="10,0,0", Scale="1,1,1") {
    rotationMode=BMM_Additive
    rotationSpace=BCS_BoneSpace
    translationMode=BMM_Additive
    translationSpace=BCS_BoneSpace
    alphaBoolBlend {blendOption=Linear}
    alphaScaleBiasClamp {interpSpeedDecreasing=8, interpSpeedIncreasing=9}
    boneToModify {boneName=dyn_coat_tair_front_r_01}
    rotation {pitch=10}
    translation {x=-0.5}
   }
   n20 = "Transform (Modify) Bone - Bone: dyn_coat_tair_front_r_01"(ComponentPose=@n19, Alpha=Get LegSwingDeltaR() -> (LegSwingDeltaR:double), Translation="-6,-5,0", Rotation="95,-30,0", Scale="1,1,1") {
    rotationMode=BMM_Additive
    rotationSpace=BCS_BoneSpace
    translationMode=BMM_Additive
    translationSpace=BCS_BoneSpace
    alphaBoolBlend {blendOption=Linear}
    alphaScaleBiasClamp {interpSpeedDecreasing=8, interpSpeedIncreasing=9}
    boneToModify {boneName=dyn_coat_tair_front_r_01}
    rotation {pitch=95, yaw=-30}
    translation=<-6 -5>
   }
   Output Pose(Result=Component To Local(ComponentPose=@n20)) {layerGroup=None}
  ]}
 EventGraph {nodes [
   Event BlueprintInitializeAnimation()
   > n1 = Cast To PlayerPawn_Athena(Object=TryGetPawnOwner())
   > Set As Player Pawn Athena(As Player Pawn Athena:Player Pawn Athena Object Reference=@n1) -> (Output_Get:Player Pawn Athena Object Reference)
   Event BlueprintUpdateAnimation()
   > n2 = Sequence!(execute) -> (then_0, then_1, then_2:Exec)
   > n3 = Branch(Condition=OR Boolean!(A=Get bIsJumping(), B=Get bIsFalling(), C:bool) -> (ReturnValue))
   > Set isJumpingNew(isJumpingNew:bool=true) -> (Output_Get:bool)
   Set isJumpingNew(execute=@n3.else, isJumpingNew=false)
   Set was_jumping(
    execute=@n2.then_1
    was_jumping:bool=Get bWasJumping!(self:Character Object Reference=Get As Player Pawn Athena() -> (As Player Pawn Athena:Player Pawn Athena Object Reference)) -> (bWasJumping:bool)
   ) -> (Output_Get:bool)
   n4 = BreakTransform(InTransform=GetDeltaTransformFromRefPose(self=GetOwnerSkeletalMeshComponent(), BoneName=thigh_l, BaseName=pelvis))
   > Set LegSwingDeltaL(LegSwingDeltaL:double=MapRangeClamped(Value=BreakRotator(InRot=@n4.Rotation).Yaw, InRangeA=-17, InRangeB=-110)) -> (Output_Get:double)
   n5 = BreakTransform(InTransform=GetDeltaTransformFromRefPose(self=GetOwnerSkeletalMeshComponent(), BoneName=thigh_r, BaseName=pelvis))
   > Set LegSwingDeltaR(LegSwingDeltaR:double=MapRangeClamped(Value=BreakRotator(InRot=@n5.Rotation).Yaw, InRangeA=-17, InRangeB=-125)) -> (Output_Get:double)
   n6 = GetOwnerSkeletalMeshComponent()
   n7 = BreakRotator(InRot=BreakTransform(InTransform=GetDeltaTransformFromRefPose(self=@n6, BoneName=thigh_r, BaseName=pelvis)).Rotation)
   > Set Pouch_swing_R(Pouch_swing_R:double=float + float(A=MapRangeClamped(Value=@n7.Yaw, InRangeA=-15, InRangeB=-120), B=0.5)) -> (Output_Get:double)
   n8 = GetOwnerSkeletalMeshComponent()
   n9 = BreakRotator(InRot=BreakTransform(InTransform=GetDeltaTransformFromRefPose(self=@n8, BoneName=thigh_r, BaseName=pelvis)).Rotation)
   > Set Pouch_swing_L(Pouch_swing_L:double=float + float(A=MapRangeClamped(Value=@n9.Yaw, InRangeA=-15, InRangeB=-120), B=0.5)) -> (Output_Get:double)
   n10 = TryGetPawnOwner()
   n11 = Get PawnVelocity()
   n12 = MapRangeClamped(
    Value=Dot Product!(
     A:Vector=GetActorForwardVector!(self:Actor Object Reference=@n10) -> (ReturnValue:Vector)
     B:Vector=Normalize!(A:Vector=@n11, Tolerance:float=0.0001) -> (ReturnValue:Vector) {is=title K2Node_CallFunction}
    ) -> (ReturnValue:double)
    InRangeA=-0.6
    InRangeB=-1
   )
   n13 = GetDeltaTransformFromRefPose(self=GetOwnerSkeletalMeshComponent(), BoneName=thigh_l, BaseName=pelvis)
   n14 = MapRangeClamped(Value=BreakRotator(InRot=BreakTransform(InTransform=@n13).Rotation).Yaw, InRangeA=-17, InRangeB=-110)
   > Set Bwd Movement(Bwd Movement:double=SelectFloat(A="Clamp (Float)"(Value=@n14, Min=0, Max=@n12), B=0, bPickA=Get bIsCrouching())) -> (Output_Get:double)
   n15 = BreakTransform(InTransform=GetDeltaTransformFromRefPose(self=GetOwnerSkeletalMeshComponent(), BoneName=head, BaseName=spine_05))
   > Set Collar_fix(Collar_fix:double=MapRangeClamped(Value=BreakRotator(InRot=@n15.Rotation).Yaw, InRangeA=0, InRangeB=-45)) -> (Output_Get:double)
   n16 = Branch(execute=@n2.then_2, Condition=AND Boolean(A=Get bIsCrouching(), B=Get bIsAccelerating2D()))
   > Set CrouchWalking(CrouchWalking:bool=true) -> (Output_Get:bool)
   Set CrouchWalking(execute=@n16.else, CrouchWalking=false)
  ]}
]
variables [
 As Player Pawn Athena {type=PlayerPawn_Athena_C}
 Bwd Movement {type=number}
 Collar_fix {type=number}
 CrouchWalking {type=boolean}
 FlapSwing_center {type=number}
 isJumpingNew {type=boolean}
 LegSwingDeltaL {type=number}
 LegSwingDeltaL_0 {type=number}
 LegSwingDeltaR {type=number}
 NewVar {type=number}
 Pouch_swing_L {type=number}
 Pouch_swing_R {type=number}
 was_jumping {type=boolean}
]
